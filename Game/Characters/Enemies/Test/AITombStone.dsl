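; AITombStone is a tombstone Blueprint under the enemy test folder that randomizes its appearance.
; Its Construction Script picks a random index across the TombStoneMeshes array and assigns that static mesh to the TombMesh component.
; The roll uses unseeded random, so the chosen mesh can change whenever the construction script reruns.
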(fn ConstructionScript
  (Components|StaticMesh|SetStaticMesh (Variables|AITombStone|GetTombMesh) (Utilities|Array|Get(acopy) (Variables|Default|GetTombStoneMeshes) (Math|Random|RandomIntegerinRange 0 (- (Utilities|Array|Length (Variables|Default|GetTombStoneMeshes)) 1)))))
